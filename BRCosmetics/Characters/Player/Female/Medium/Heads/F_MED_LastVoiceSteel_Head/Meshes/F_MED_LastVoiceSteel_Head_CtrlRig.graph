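# Control Rig for the LastVoiceSteel female medium head that applies a small procedural corrective rotation to the neck as the head moves.
# A spherical pose reader measures the head bone's orientation relative to spine_05, with debug drawing left on.
# Its output scales a Z Euler rotation of up to 10 degrees, applied to neck_02 through Modify Transforms.
# The hierarchy also carries facial bones and expression curves, which the graph leaves untouched.
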ControlRigBlueprint /BRCosmetics/Characters/Player/Female/Medium/Heads/F_MED_LastVoiceSteel_Head/Meshes/F_MED_LastVoiceSteel_Head_CtrlRig
graphs [
 RigVMFunctionLibrary {links [], nodes []}
 RigVMModel {nodes [
   RigUnit_BeginExecution = &RigUnit_BeginExecution()
   n1 = &RigUnit_SphericalPoseReader(
    ActiveRegionScaleFactors=(PositiveWidth=0.1,NegativeWidth=0.1,PositiveHeight=0.1,NegativeHeight=0.1)
    ActiveRegionSize=0.1
    Debug=(bDrawDebug=True,DebugScale=25,DebugSegments=20,DebugThickness=0.25)
    DriverAxis=(X=0,Y=-1,Z=0)
    DriverItem=(Type=Bone,Name="head")
    FalloffRegionScaleFactors=(PositiveWidth=0.5,NegativeWidth=0.5,PositiveHeight=0.5,NegativeHeight=0.5)
    FalloffSize=0.5
    FlipWidthScaling=true
    OptionalParentItem=(Type=Bone,Name="spine_05")
    RotationOffset=(X=90,Y=180,Z=90)
    ExecutePin=@RigUnit_BeginExecution.ExecutePin
   )
   > &RigUnit_ModifyTransforms(
    ItemToModify=((Item=(Type=Bone,Name="neck_02"),Transform=(Rotation=(X=0,Y=0,Z=0,W=1),Translation=(X=0,Y=0,Z=0),Scale3D=(X=1,Y=1,Z=1))))
    Weight=1
    ItemToModify.0.Transform.Rotation=&RigVMFunction_MathQuaternionFromEuler(Euler=(X=0,Y=0,Z=10), Euler.Z=&RigVMFunction_MathDoubleMul(A=10, B=@n1.OutputParam).Result).Result
   )
  ]}
]
hierarchy [
 root {loc=<0 0 0>, rot=<0 0 0 1>}
 pelvis {parent=root, loc=<0 -0.5301 98.0571>, rot=<0.0127 0.707 -0.0127 -0.707>}
 spine_01 {parent=pelvis, loc=<6.1208 -0.9328 0>, rot=<0 0 0.034 -0.9994>}
 spine_02 {parent=spine_01, loc=<6.5996 -0.2095 0>, rot=<0 0 0.0238 -0.9997>}
 spine_03 {parent=spine_02, loc=<6.4914 -0.1237 0>, rot=<0 0 0.0149 -0.9999>}
 spine_04 {parent=spine_03, loc=<6.9598 0.0794 0>, rot=<0 0 -0.0132 -0.9999>}
 spine_05 {parent=spine_04, loc=<6.9987 0 0>, rot=<0 0 -0.0015 -1>}
 neck_01 {parent=spine_05, loc=<11.3138 0.1529 0>, rot=<0 0 -0.2003 -0.9797>}
 neck_02 {parent=neck_01, loc=<4.5627 0 0>, rot=<0 0 0.0504 -0.9987>}
 head {parent=neck_02, loc=<5.3914 0 0>, rot=<0 0 0.115 -0.9934>}
 FACIAL_C_FacialRoot {parent=head, loc=<2.5547 -0.8284 -0.0001>, rot=<0.5021 -0.4979 -0.5021 -0.4979>}
 FACIAL_C_Forehead {parent=FACIAL_C_FacialRoot, loc=<-0.0001 -3.8084 9.1562>, rot=<-0.0436 0 0 -0.999>}
 FACIAL_L_ForeheadIn {parent=FACIAL_C_FacialRoot, loc=<2.0308 -3.7689 8.7077>, rot=<-0.0609 -0.0696 0.0043 -0.9957>}
 FACIAL_R_ForeheadIn {parent=FACIAL_C_FacialRoot, loc=<-2.0308 -3.7689 8.7077>, rot=<-0.0609 0.0696 -0.0043 -0.9957>}
 FACIAL_L_ForeheadMid {parent=FACIAL_C_FacialRoot, loc=<3.6634 -4.3575 7.7584>, rot=<-0.0712 -0.1757 -0.0206 -0.9816>}
 FACIAL_R_ForeheadMid {parent=FACIAL_C_FacialRoot, loc=<-3.6634 -4.3575 7.7584>, rot=<-0.0712 0.1757 0.0206 -0.9816>}
 FACIAL_L_ForeheadOut {parent=FACIAL_C_FacialRoot, loc=<5.0722 -4.0351 6.5977>, rot=<-0.1215 -0.2557 -0.0792 -0.9558>}
 FACIAL_R_ForeheadOut {parent=FACIAL_C_FacialRoot, loc=<-5.0722 -4.0351 6.5977>, rot=<-0.1215 0.2557 0.0792 -0.9558>}
 FACIAL_L_EyesackUpper {parent=FACIAL_C_FacialRoot, loc=<3.2601 -2.588 8.2688>, rot=<0.0487 -0.0843 -0.0646 -0.9932>}
 FACIAL_R_EyesackUpper {parent=FACIAL_C_FacialRoot, loc=<-3.2601 -2.588 8.2688>, rot=<0.0487 0.0843 0.0646 -0.9932>}
 FACIAL_L_EyelidUpperFurrow {parent=FACIAL_C_FacialRoot, loc=<3.0826 -2.5368 8.1335>, rot=<-0.0301 -0.1509 -0.0797 -0.9849>}
 FACIAL_R_EyelidUpperFurrow {parent=FACIAL_C_FacialRoot, loc=<-3.0826 -2.5368 8.1335>, rot=<-0.0301 0.1509 0.0797 -0.9849>}
 FACIAL_L_EyelidUpperA {parent=FACIAL_C_FacialRoot, loc=<2.9775 -0.6819 5.6843>, rot=<-0.1478 0 0 -0.989>}
 FACIAL_R_EyelidUpperA {parent=FACIAL_C_FacialRoot, loc=<-2.9775 -0.6819 5.6843>, rot=<-0.1478 0 0 -0.989>}
 FACIAL_L_Eye {parent=FACIAL_C_FacialRoot, loc=<2.9775 -0.6819 5.6843>, rot=<0 0 0 1>}
 FACIAL_R_Eye {parent=FACIAL_C_FacialRoot, loc=<-2.9775 -0.6819 5.6843>, rot=<0 0 0 1>}
 FACIAL_L_EyelidLowerA {parent=FACIAL_C_FacialRoot, loc=<2.9775 -0.6819 5.6843>, rot=<0.284 0 0 -0.9588>}
 FACIAL_R_EyelidLowerA {parent=FACIAL_C_FacialRoot, loc=<-2.9775 -0.6819 5.6843>, rot=<0.284 0 0 -0.9588>}
 FACIAL_L_EyeCornerInner {parent=FACIAL_C_FacialRoot, loc=<1.8754 0.3356 7.5103>, rot=<0 0 0 -1>}
 FACIAL_R_EyeCornerInner {parent=FACIAL_C_FacialRoot, loc=<-1.8754 0.3356 7.5103>, rot=<0 0 0 -1>}
 FACIAL_L_EyeCornerOuter {parent=FACIAL_C_FacialRoot, loc=<5.1907 -1.249 6.326>, rot=<0 -0.3827 0 -0.9239>}
 FACIAL_R_EyeCornerOuter {parent=FACIAL_C_FacialRoot, loc=<-5.1907 -1.2489 6.3263>, rot=<0 0.3827 0 -0.9239>}
 FACIAL_L_EyesackLower {parent=FACIAL_C_FacialRoot, loc=<3.8339 0.9757 6.6241>, rot=<-0.0223 -0.1398 -0.0223 -0.9897>}
 FACIAL_R_EyesackLower {parent=FACIAL_C_FacialRoot, loc=<-3.8339 0.9757 6.6241>, rot=<-0.0223 0.1398 0.0223 -0.9897>}
 FACIAL_L_CheekInner {parent=FACIAL_C_FacialRoot, loc=<2.2791 1.6442 7.4156>, rot=<-0.1233 -0.088 -0.1972 -0.9686>}
 FACIAL_R_CheekInner {parent=FACIAL_C_FacialRoot, loc=<-2.2791 1.6441 7.4156>, rot=<-0.1233 0.088 0.1972 -0.9686>}
 FACIAL_L_CheekOuter {parent=FACIAL_C_FacialRoot, loc=<4.0812 1.3725 5.6125>, rot=<0.0551 -0.377 0.1003 -0.9191>}
 FACIAL_R_CheekOuter {parent=FACIAL_C_FacialRoot, loc=<-4.0812 1.3725 5.6125>, rot=<0.0551 0.377 -0.1003 -0.9191>}
 FACIAL_L_NasolabialBulge {parent=FACIAL_C_FacialRoot, loc=<3.0191 4.1779 6.1457>, rot=<0.0579 -0.3774 0.0238 -0.9239>}
 FACIAL_R_NasolabialBulge {parent=FACIAL_C_FacialRoot, loc=<-3.0191 4.1779 6.1457>, rot=<0.0579 0.3774 -0.0238 -0.9239>}
 FACIAL_L_NasolabialFurrow {parent=FACIAL_C_FacialRoot, loc=<2.5402 3.3924 8.6275>, rot=<-0.0333 -0.1258 -0.3638 -0.9223>}
 FACIAL_R_NasolabialFurrow {parent=FACIAL_C_FacialRoot, loc=<-2.5402 3.3924 8.6275>, rot=<-0.0333 0.1258 0.3638 -0.9223>}
 FACIAL_L_CheekLower {parent=FACIAL_C_FacialRoot, loc=<3.2684 3.5251 5.8845>, rot=<0.0345 -0.427 0.0577 -0.9018>}
 FACIAL_R_CheekLower {parent=FACIAL_C_FacialRoot, loc=<-3.2684 3.5251 5.8845>, rot=<0.0345 0.427 -0.0577 -0.9018>}
 FACIAL_C_Nose {parent=FACIAL_C_FacialRoot, loc=<-0.0001 0.8881 8.9042>, rot=<-0.1305 0 0 -0.9914>}
 FACIAL_C_NoseLower {parent=FACIAL_C_Nose, loc=<0 2.7373 0.0359>, rot=<0.1305 0 0 -0.9914>}
 FACIAL_L_Nostril {parent=FACIAL_C_Nose, loc=<0.7057 2.3815 0.3769>, rot=<0.0173 0.1305 0.0023 -0.9913>}
 FACIAL_R_Nostril {parent=FACIAL_C_Nose, loc=<-0.7057 2.3815 0.3769>, rot=<0.0173 -0.1305 -0.0023 -0.9913>}
 FACIAL_C_NoseTip {parent=FACIAL_C_Nose, loc=<0 1.941 0.5967>, rot=<0.0872 0 0 -0.9962>}
 FACIAL_C_MouthUpper {parent=FACIAL_C_FacialRoot, loc=<-0.0001 4.304 7.3484>, rot=<0 0 0 -1>}
 FACIAL_C_LipUpper {parent=FACIAL_C_MouthUpper, loc=<0 0.2435 1.9074>, rot=<-0.0872 0 0 -0.9962>}
 FACIAL_L_LipUpper {parent=FACIAL_C_MouthUpper, loc=<0.6469 0.3104 1.9248>, rot=<-0.0778 -0.1601 -0.0379 -0.9833>}
 FACIAL_R_LipUpper {parent=FACIAL_C_MouthUpper, loc=<-0.6469 0.3104 1.9248>, rot=<-0.0778 0.1601 0.0379 -0.9833>}
 FACIAL_L_LipUpperOuter {parent=FACIAL_C_MouthUpper, loc=<1.3094 0.3047 1.5503>, rot=<-0.0983 -0.2294 -0.0983 -0.9634>}
 FACIAL_R_LipUpperOuter {parent=FACIAL_C_MouthUpper, loc=<-1.3094 0.3047 1.5503>, rot=<-0.0983 0.2294 0.0983 -0.9634>}
 FACIAL_L_LipCorner {parent=FACIAL_C_MouthUpper, loc=<1.9469 0.5034 0.9049>, rot=<0 -0.2079 0 -0.9781>}
 FACIAL_R_LipCorner {parent=FACIAL_C_MouthUpper, loc=<-1.9469 0.5034 0.9049>, rot=<0 0.2079 0 -0.9781>}
 FACIAL_L_Masseter {parent=FACIAL_C_FacialRoot, loc=<5.8814 1.7644 3.5581>, rot=<0.0693 0.1129 0.0079 -0.9912>}
 FACIAL_R_Masseter {parent=FACIAL_C_FacialRoot, loc=<-5.8814 1.7643 3.5581>, rot=<0.0693 -0.1129 -0.0079 -0.9912>}
 FACIAL_C_UnderChin {parent=FACIAL_C_FacialRoot, loc=<-0.0001 7.9213 4.148>, rot=<0.5 0 0 -0.866>}
 FACIAL_L_UnderChin {parent=FACIAL_C_FacialRoot, loc=<2.2655 6.987 2.9681>, rot=<0.4025 -0.355 0.0823 -0.8398>}
 FACIAL_R_UnderChin {parent=FACIAL_C_FacialRoot, loc=<-2.2655 6.987 2.9681>, rot=<0.4025 0.355 -0.0823 -0.8398>}
 FACIAL_C_TeethUpper {parent=FACIAL_C_FacialRoot, loc=<-0.0001 4.6605 8.0097>, rot=<0 0 0 1>}
 FACIAL_C_LowerLipRotation {parent=FACIAL_C_FacialRoot, loc=<-0.0001 2.1449 1.6401>, rot=<0.342 0 0 -0.9397>}
 FACIAL_C_MouthLower {parent=FACIAL_C_LowerLipRotation, loc=<0 -2.0152 5.7607>, rot=<-0.342 0 0 -0.9397>}
 FACIAL_C_LipLower {parent=FACIAL_C_MouthLower, loc=<0 1.0403 1.7672>, rot=<0.0872 0 0 -0.9962>}
 FACIAL_L_LipLower {parent=FACIAL_C_MouthLower, loc=<0.6772 1.0475 1.6475>, rot=<0.0927 -0.1295 0.0273 -0.9869>}
 FACIAL_R_LipLower {parent=FACIAL_C_MouthLower, loc=<-0.6772 1.0475 1.6475>, rot=<0.0927 0.1295 -0.0273 -0.9869>}
 FACIAL_L_LipLowerOuter {parent=FACIAL_C_MouthLower, loc=<1.3474 0.9599 1.357>, rot=<0.0913 -0.2322 0.0318 -0.9679>}
 FACIAL_R_LipLowerOuter {parent=FACIAL_C_MouthLower, loc=<-1.3474 0.9599 1.357>, rot=<0.0913 0.2322 -0.0318 -0.9679>}
 FACIAL_C_TeethLower {parent=FACIAL_C_LowerLipRotation, loc=<-0.0008 -2.1009 6.4351>, rot=<-0.342 0 0 -0.9397>}
 FACIAL_C_Tongue1 {parent=FACIAL_C_TeethLower, loc=<0.001 0.0817 -5.0785>, rot=<-0.2588 0 0 -0.9659>}
 FACIAL_C_Tongue2 {parent=FACIAL_C_Tongue1, loc=<-0.0002 0.4571 1.4859>, rot=<0.2588 0 0 -0.9659>}
 FACIAL_C_Tongue3 {parent=FACIAL_C_Tongue2, loc=<-0.0001 0.2288 1.3297>, rot=<0.1305 0 0 -0.9914>}
 FACIAL_C_Tongue4 {parent=FACIAL_C_Tongue3, loc=<-0.0001 -0.1598 1.1009>, rot=<0.0436 0 0 -0.999>}
 FACIAL_C_TongueUpper3 {parent=FACIAL_C_Tongue3, loc=<0 -0.2941 -0.2398>, rot=<0 0 0 1>}
 FACIAL_C_TongueLower3 {parent=FACIAL_C_Tongue3, loc=<0 0.8294 0.0462>, rot=<0 0 0 1>}
 FACIAL_L_TongueSide3 {parent=FACIAL_C_Tongue3, loc=<1.4941 -0.0273 0.0614>, rot=<0 0 0 1>}
 FACIAL_R_TongueSide3 {parent=FACIAL_C_Tongue3, loc=<-1.4942 -0.0273 0.0614>, rot=<0 0 0 1>}
 FACIAL_C_TongueUpper2 {parent=FACIAL_C_Tongue2, loc=<-0.0001 -0.5618 -0.3851>, rot=<0 0 0 1>}
 FACIAL_L_TongueSide2 {parent=FACIAL_C_Tongue2, loc=<1.7814 -0.0872 -0.0014>, rot=<0 0 0 1>}
 FACIAL_R_TongueSide2 {parent=FACIAL_C_Tongue2, loc=<-1.7814 -0.0872 -0.0014>, rot=<0 0 0 1>}
 FACIAL_C_Jaw {parent=FACIAL_C_FacialRoot, loc=<-0.0001 2.1449 1.6401>, rot=<0.342 0 0 -0.9397>}
 FACIAL_C_Jawline {parent=FACIAL_C_Jaw, loc=<0 1.0897 7.3181>, rot=<0.2588 0 0 -0.9659>}
 FACIAL_L_ChinSide {parent=FACIAL_C_Jaw, loc=<2.392 -0.5846 7.0377>, rot=<-0.2207 -0.3739 0.05 -0.8994>}
 FACIAL_R_ChinSide {parent=FACIAL_C_Jaw, loc=<-2.392 -0.5846 7.0377>, rot=<-0.2207 0.3739 -0.05 -0.8994>}
 FACIAL_C_Chin {parent=FACIAL_C_Jaw, loc=<0 -0.1679 7.0336>, rot=<-0.2588 0 0 -0.9659>}
 FACIAL_L_Ear {parent=FACIAL_C_FacialRoot, loc=<6.414 -0.3149 -0.3525>, rot=<-0.0262 -0.0007 -0.0262 -0.9993>}
 FACIAL_R_Ear {parent=FACIAL_C_FacialRoot, loc=<-6.414 -0.3149 -0.3525>, rot=<-0.0262 0.0007 0.0262 -0.9993>}
 FACIAL_L_Jawline {parent=FACIAL_C_FacialRoot, loc=<3.7908 4.9604 4.6194>, rot=<0.2092 0.1578 -0.0547 -0.9635>}
 FACIAL_R_Jawline {parent=FACIAL_C_FacialRoot, loc=<-3.7908 4.9604 4.6194>, rot=<0.2092 -0.1578 0.0547 -0.9635>}
 FACIAL_C_NoseBridge {parent=FACIAL_C_FacialRoot, loc=<-0.0001 -1.4599 9.4799>, rot=<0.0436 0 0 -0.999>}
 FACIAL_L_EyelidUpperB {parent=FACIAL_C_FacialRoot, loc=<2.9775 -0.6819 5.6843>, rot=<-0.225 0 0 -0.9744>}
 FACIAL_L_EyelidLowerB {parent=FACIAL_C_FacialRoot, loc=<2.9775 -0.6819 5.6843>, rot=<0.3173 0 0 -0.9483>}
 FACIAL_R_EyelidLowerB {parent=FACIAL_C_FacialRoot, loc=<-2.9775 -0.6819 5.6843>, rot=<0.3256 0 0 -0.9455>}
 FACIAL_R_EyelidUpperB {parent=FACIAL_C_FacialRoot, loc=<-2.9775 -0.6819 5.6843>, rot=<-0.225 0 0 -0.9744>}
 clavicle_l {parent=spine_05, loc=<5.4831 -0.1053 -3.2946>, rot=<0.0465 -0.7289 -0.009 -0.6829>}
 upperarm_l {parent=clavicle_l, loc=<9.8658 -1.2136 -0.4105>, rot=<0.0164 -0.3245 -0.0269 -0.9454>}
 clavicle_r {parent=spine_05, loc=<5.4831 -0.1053 3.2946>, rot=<0.7289 0.0465 -0.6829 0.009>}
 upperarm_r {parent=clavicle_r, loc=<-9.8658 1.2136 0.4105>, rot=<0.0164 -0.3245 -0.0269 -0.9454>}
 CTRL_expressions_browDownL {type=Curve, loc=<0 0 0>, rot=<0 0 0 1>}
 CTRL_expressions_browDownR {type=Curve, loc=<0 0 0>, rot=<0 0 0 1>}
 CTRL_expressions_browLateralL {type=Curve, loc=<0 0 0>, rot=<0 0 0 1>}
 CTRL_expressions_browLateralR {type=Curve, loc=<0 0 0>, rot=<0 0 0 1>}
 CTRL_expressions_browRaiseInL {type=Curve, loc=<0 0 0>, rot=<0 0 0 1>}
 CTRL_expressions_browRaiseInR {type=Curve, loc=<0 0 0>, rot=<0 0 0 1>}
 CTRL_expressions_browRaiseOuterL {type=Curve, loc=<0 0 0>, rot=<0 0 0 1>}
 CTRL_expressions_browRaiseOuterR {type=Curve, loc=<0 0 0>, rot=<0 0 0 1>}
 CTRL_expressions_browUINL {type=Curve, loc=<0 0 0>, rot=<0 0 0 1>}
 CTRL_expressions_browUINR {type=Curve, loc=<0 0 0>, rot=<0 0 0 1>}
 CTRL_expressions_browUOUTL {type=Curve, loc=<0 0 0>, rot=<0 0 0 1>}
 CTRL_expressions_browUOUTR {type=Curve, loc=<0 0 0>, rot=<0 0 0 1>}
 CTRL_expressions_browDINL {type=Curve, loc=<0 0 0>, rot=<0 0 0 1>}
 CTRL_expressions_browDINR {type=Curve, loc=<0 0 0>, rot=<0 0 0 1>}
 CTRL_expressions_browDOUTL {type=Curve, loc=<0 0 0>, rot=<0 0 0 1>}
 CTRL_expressions_browDOUTR {type=Curve, loc=<0 0 0>, rot=<0 0 0 1>}
 CTRL_expressions_eyeBlinkL {type=Curve, loc=<0 0 0>, rot=<0 0 0 1>}
 CTRL_expressions_eyeBlinkR {type=Curve, loc=<0 0 0>, rot=<0 0 0 1>}
 CTRL_expressions_eyeWidenL {type=Curve, loc=<0 0 0>, rot=<0 0 0 1>}
 CTRL_expressions_eyeWidenR {type=Curve, loc=<0 0 0>, rot=<0 0 0 1>}
 CTRL_expressions_eyeSquintInnerL {type=Curve, loc=<0 0 0>, rot=<0 0 0 1>}
 CTRL_expressions_eyeSquintInnerR {type=Curve, loc=<0 0 0>, rot=<0 0 0 1>}
 CTRL_expressions_eyeCheekRaiseL {type=Curve, loc=<0 0 0>, rot=<0 0 0 1>}
 CTRL_expressions_eyeCheekRaiseR {type=Curve, loc=<0 0 0>, rot=<0 0 0 1>}
 CTRL_expressions_eyeUpperLidUpL {type=Curve, loc=<0 0 0>, rot=<0 0 0 1>}
 CTRL_expressions_eyeUpperLidUpR {type=Curve, loc=<0 0 0>, rot=<0 0 0 1>}
 CTRL_expressions_eyeRelaxL {type=Curve, loc=<0 0 0>, rot=<0 0 0 1>}
 CTRL_expressions_eyeRelaxR {type=Curve, loc=<0 0 0>, rot=<0 0 0 1>}
 CTRL_expressions_eyeLowerLidUpL {type=Curve, loc=<0 0 0>, rot=<0 0 0 1>}
 CTRL_expressions_eyeLowerLidUpR {type=Curve, loc=<0 0 0>, rot=<0 0 0 1>}
 CTRL_expressions_eyeLowerLidDownL {type=Curve, loc=<0 0 0>, rot=<0 0 0 1>}
 CTRL_expressions_eyeLowerLidDownR {type=Curve, loc=<0 0 0>, rot=<0 0 0 1>}
 CTRL_expressions_eyelidRollPosUL {type=Curve, loc=<0 0 0>, rot=<0 0 0 1>}
 CTRL_expressions_eyelidRollPosUR {type=Curve, loc=<0 0 0>, rot=<0 0 0 1>}
 CTRL_expressions_eyelidRollPosDL {type=Curve, loc=<0 0 0>, rot=<0 0 0 1>}
 CTRL_expressions_eyelidRollPosDR {type=Curve, loc=<0 0 0>, rot=<0 0 0 1>}
 CTRL_expressions_eyelidRollNegUL {type=Curve, loc=<0 0 0>, rot=<0 0 0 1>}
 CTRL_expressions_eyelidRollNegUR {type=Curve, loc=<0 0 0>, rot=<0 0 0 1>}
 CTRL_expressions_eyelidRollNegDL {type=Curve, loc=<0 0 0>, rot=<0 0 0 1>}
 CTRL_expressions_eyelidRollNegDR {type=Curve, loc=<0 0 0>, rot=<0 0 0 1>}
 CTRL_expressions_eyeLookUpL {type=Curve, loc=<0 0 0>, rot=<0 0 0 1>}
 CTRL_expressions_eyeLookUpR {type=Curve, loc=<0 0 0>, rot=<0 0 0 1>}
 CTRL_expressions_eyeLookDownL {type=Curve, loc=<0 0 0>, rot=<0 0 0 1>}
 CTRL_expressions_eyeLookDownR {type=Curve, loc=<0 0 0>, rot=<0 0 0 1>}
 CTRL_expressions_eyeLookLeftL {type=Curve, loc=<0 0 0>, rot=<0 0 0 1>}
 CTRL_expressions_eyeLookLeftR {type=Curve, loc=<0 0 0>, rot=<0 0 0 1>}
 CTRL_expressions_eyeLookRightL {type=Curve, loc=<0 0 0>, rot=<0 0 0 1>}
 CTRL_expressions_eyeLookRightR {type=Curve, loc=<0 0 0>, rot=<0 0 0 1>}
 CTRL_expressions_eyeEyelidsReduce {type=Curve, loc=<0 0 0>, rot=<0 0 0 1>}
 CTRL_expressions_noseWrinkleL {type=Curve, loc=<0 0 0>, rot=<0 0 0 1>}
 CTRL_expressions_noseWrinkleR {type=Curve, loc=<0 0 0>, rot=<0 0 0 1>}
 CTRL_expressions_noseWrinkleUpperL {type=Curve, loc=<0 0 0>, rot=<0 0 0 1>}
 CTRL_expressions_noseWrinkleUpperR {type=Curve, loc=<0 0 0>, rot=<0 0 0 1>}
 CTRL_expressions_noseNostrilDepressL {type=Curve, loc=<0 0 0>, rot=<0 0 0 1>}
 CTRL_expressions_noseNostrilDepressR {type=Curve, loc=<0 0 0>, rot=<0 0 0 1>}
 CTRL_expressions_noseNostrilDilateL {type=Curve, loc=<0 0 0>, rot=<0 0 0 1>}
 CTRL_expressions_noseNostrilDilateR {type=Curve, loc=<0 0 0>, rot=<0 0 0 1>}
 CTRL_expressions_noseNostrilCompressL {type=Curve, loc=<0 0 0>, rot=<0 0 0 1>}
 CTRL_expressions_noseNostrilCompressR {type=Curve, loc=<0 0 0>, rot=<0 0 0 1>}
 CTRL_expressions_noseRotateL {type=Curve, loc=<0 0 0>, rot=<0 0 0 1>}
 CTRL_expressions_noseRotateR {type=Curve, loc=<0 0 0>, rot=<0 0 0 1>}
 CTRL_expressions_noseSquash {type=Curve, loc=<0 0 0>, rot=<0 0 0 1>}
 CTRL_expressions_noseStretch {type=Curve, loc=<0 0 0>, rot=<0 0 0 1>}
 CTRL_expressions_mouthCheekSuckL {type=Curve, loc=<0 0 0>, rot=<0 0 0 1>}
 CTRL_expressions_mouthCheekSuckR {type=Curve, loc=<0 0 0>, rot=<0 0 0 1>}
 CTRL_expressions_mouthCheekBlowL {type=Curve, loc=<0 0 0>, rot=<0 0 0 1>}
 CTRL_expressions_mouthCheekBlowR {type=Curve, loc=<0 0 0>, rot=<0 0 0 1>}
 CTRL_expressions_mouthLipsBlowL {type=Curve, loc=<0 0 0>, rot=<0 0 0 1>}
 CTRL_expressions_mouthLipsBlowR {type=Curve, loc=<0 0 0>, rot=<0 0 0 1>}
 CTRL_expressions_mouthLeft {type=Curve, loc=<0 0 0>, rot=<0 0 0 1>}
 CTRL_expressions_mouthRight {type=Curve, loc=<0 0 0>, rot=<0 0 0 1>}
 CTRL_expressions_mouthUp {type=Curve, loc=<0 0 0>, rot=<0 0 0 1>}
 CTRL_expressions_mouthDown {type=Curve, loc=<0 0 0>, rot=<0 0 0 1>}
 CTRL_expressions_mouthUpperLipRaiseL {type=Curve, loc=<0 0 0>, rot=<0 0 0 1>}
 CTRL_expressions_mouthUpperLipRaiseR {type=Curve, loc=<0 0 0>, rot=<0 0 0 1>}
 CTRL_expressions_mouthLowerLipDepressL {type=Curve, loc=<0 0 0>, rot=<0 0 0 1>}
 CTRL_expressions_mouthLowerLipDepressR {type=Curve, loc=<0 0 0>, rot=<0 0 0 1>}
 CTRL_expressions_mouthCornerPullL {type=Curve, loc=<0 0 0>, rot=<0 0 0 1>}
 CTRL_expressions_mouthCornerPullR {type=Curve, loc=<0 0 0>, rot=<0 0 0 1>}
 CTRL_expressions_mouthStretchL {type=Curve, loc=<0 0 0>, rot=<0 0 0 1>}
 CTRL_expressions_mouthStretchR {type=Curve, loc=<0 0 0>, rot=<0 0 0 1>}
 CTRL_expressions_mouthDimpleL {type=Curve, loc=<0 0 0>, rot=<0 0 0 1>}
 CTRL_expressions_mouthDimpleR {type=Curve, loc=<0 0 0>, rot=<0 0 0 1>}
 CTRL_expressions_mouthCornerDepressL {type=Curve, loc=<0 0 0>, rot=<0 0 0 1>}
 CTRL_expressions_mouthCornerDepressR {type=Curve, loc=<0 0 0>, rot=<0 0 0 1>}
 CTRL_expressions_mouthPressUL {type=Curve, loc=<0 0 0>, rot=<0 0 0 1>}
 CTRL_expressions_mouthPressUR {type=Curve, loc=<0 0 0>, rot=<0 0 0 1>}
 CTRL_expressions_mouthPressDL {type=Curve, loc=<0 0 0>, rot=<0 0 0 1>}
 CTRL_expressions_mouthPressDR {type=Curve, loc=<0 0 0>, rot=<0 0 0 1>}
 CTRL_expressions_mouthLipsPurseUL {type=Curve, loc=<0 0 0>, rot=<0 0 0 1>}
 CTRL_expressions_mouthLipsPurseUR {type=Curve, loc=<0 0 0>, rot=<0 0 0 1>}
 CTRL_expressions_mouthLipsPurseDL {type=Curve, loc=<0 0 0>, rot=<0 0 0 1>}
 CTRL_expressions_mouthLipsPurseDR {type=Curve, loc=<0 0 0>, rot=<0 0 0 1>}
 CTRL_expressions_mouthLipsTowardsUL {type=Curve, loc=<0 0 0>, rot=<0 0 0 1>}
 CTRL_expressions_mouthLipsTowardsUR {type=Curve, loc=<0 0 0>, rot=<0 0 0 1>}
 CTRL_expressions_mouthLipsTowardsDL {type=Curve, loc=<0 0 0>, rot=<0 0 0 1>}
 CTRL_expressions_mouthLipsTowardsDR {type=Curve, loc=<0 0 0>, rot=<0 0 0 1>}
 CTRL_expressions_mouthFunnelUL {type=Curve, loc=<0 0 0>, rot=<0 0 0 1>}
 CTRL_expressions_mouthFunnelUR {type=Curve, loc=<0 0 0>, rot=<0 0 0 1>}
 CTRL_expressions_mouthFunnelDL {type=Curve, loc=<0 0 0>, rot=<0 0 0 1>}
 CTRL_expressions_mouthFunnelDR {type=Curve, loc=<0 0 0>, rot=<0 0 0 1>}
 CTRL_expressions_mouthLipsTogetherUL {type=Curve, loc=<0 0 0>, rot=<0 0 0 1>}
 CTRL_expressions_mouthLipsTogetherUR {type=Curve, loc=<0 0 0>, rot=<0 0 0 1>}
 CTRL_expressions_mouthLipsTogetherDL {type=Curve, loc=<0 0 0>, rot=<0 0 0 1>}
 CTRL_expressions_mouthLipsTogetherDR {type=Curve, loc=<0 0 0>, rot=<0 0 0 1>}
 CTRL_expressions_mouthUpperLipBiteL {type=Curve, loc=<0 0 0>, rot=<0 0 0 1>}
 CTRL_expressions_mouthUpperLipBiteR {type=Curve, loc=<0 0 0>, rot=<0 0 0 1>}
 CTRL_expressions_mouthLowerLipBiteL {type=Curve, loc=<0 0 0>, rot=<0 0 0 1>}
 CTRL_expressions_mouthLowerLipBiteR {type=Curve, loc=<0 0 0>, rot=<0 0 0 1>}
 CTRL_expressions_mouthCornerSharpenUL {type=Curve, loc=<0 0 0>, rot=<0 0 0 1>}
 CTRL_expressions_mouthCornerSharpenUR {type=Curve, loc=<0 0 0>, rot=<0 0 0 1>}
 CTRL_expressions_mouthCornerSharpenDL {type=Curve, loc=<0 0 0>, rot=<0 0 0 1>}
 CTRL_expressions_mouthCornerSharpenDR {type=Curve, loc=<0 0 0>, rot=<0 0 0 1>}
 CTRL_expressions_mouthCornerRounderUL {type=Curve, loc=<0 0 0>, rot=<0 0 0 1>}
 CTRL_expressions_mouthCornerRounderUR {type=Curve, loc=<0 0 0>, rot=<0 0 0 1>}
 CTRL_expressions_mouthCornerRounderDL {type=Curve, loc=<0 0 0>, rot=<0 0 0 1>}
 CTRL_expressions_mouthCornerRounderDR {type=Curve, loc=<0 0 0>, rot=<0 0 0 1>}
 CTRL_expressions_mouthUpperLipTowardsTeethL {type=Curve, loc=<0 0 0>, rot=<0 0 0 1>}
 CTRL_expressions_mouthUpperLipTowardsTeethR {type=Curve, loc=<0 0 0>, rot=<0 0 0 1>}
 CTRL_expressions_mouthLowerLipTowardsTeethL {type=Curve, loc=<0 0 0>, rot=<0 0 0 1>}
 CTRL_expressions_mouthLowerLipTowardsTeethR {type=Curve, loc=<0 0 0>, rot=<0 0 0 1>}
 CTRL_expressions_mouthCornerUpL {type=Curve, loc=<0 0 0>, rot=<0 0 0 1>}
 CTRL_expressions_mouthCornerUpR {type=Curve, loc=<0 0 0>, rot=<0 0 0 1>}
 CTRL_expressions_mouthCornerDownL {type=Curve, loc=<0 0 0>, rot=<0 0 0 1>}
 CTRL_expressions_mouthCornerDownR {type=Curve, loc=<0 0 0>, rot=<0 0 0 1>}
 CTRL_expressions_mouthCornerWideL {type=Curve, loc=<0 0 0>, rot=<0 0 0 1>}
 CTRL_expressions_mouthCornerWideR {type=Curve, loc=<0 0 0>, rot=<0 0 0 1>}
 CTRL_expressions_mouthCornerNarrowL {type=Curve, loc=<0 0 0>, rot=<0 0 0 1>}
 CTRL_expressions_mouthCornerNarrowR {type=Curve, loc=<0 0 0>, rot=<0 0 0 1>}
 CTRL_expressions_tongueUp {type=Curve, loc=<0 0 0>, rot=<0 0 0 1>}
 CTRL_expressions_tongueDown {type=Curve, loc=<0 0 0>, rot=<0 0 0 1>}
 CTRL_expressions_tongueLeft {type=Curve, loc=<0 0 0>, rot=<0 0 0 1>}
 CTRL_expressions_tongueRight {type=Curve, loc=<0 0 0>, rot=<0 0 0 1>}
 CTRL_expressions_tongueOut {type=Curve, loc=<0 0 0>, rot=<0 0 0 1>}
 CTRL_expressions_tongueIn {type=Curve, loc=<0 0 0>, rot=<0 0 0 1>}
 CTRL_expressions_tongueRollUp {type=Curve, loc=<0 0 0>, rot=<0 0 0 1>}
 CTRL_expressions_tongueRollDown {type=Curve, loc=<0 0 0>, rot=<0 0 0 1>}
 CTRL_expressions_tongueRollLeft {type=Curve, loc=<0 0 0>, rot=<0 0 0 1>}
 CTRL_expressions_tongueRollRight {type=Curve, loc=<0 0 0>, rot=<0 0 0 1>}
 CTRL_expressions_tongueWide {type=Curve, loc=<0 0 0>, rot=<0 0 0 1>}
 CTRL_expressions_tongueNarrow {type=Curve, loc=<0 0 0>, rot=<0 0 0 1>}
 CTRL_expressions_jawOpen {type=Curve, loc=<0 0 0>, rot=<0 0 0 1>}
 CTRL_expressions_jawLeft {type=Curve, loc=<0 0 0>, rot=<0 0 0 1>}
 CTRL_expressions_jawRight {type=Curve, loc=<0 0 0>, rot=<0 0 0 1>}
 CTRL_expressions_jawFwd {type=Curve, loc=<0 0 0>, rot=<0 0 0 1>}
 CTRL_expressions_jawBack {type=Curve, loc=<0 0 0>, rot=<0 0 0 1>}
 CTRL_expressions_jawChinRaiseDL {type=Curve, loc=<0 0 0>, rot=<0 0 0 1>}
 CTRL_expressions_jawChinRaiseDR {type=Curve, loc=<0 0 0>, rot=<0 0 0 1>}
 CTRL_expressions_jawChinRaiseUL {type=Curve, loc=<0 0 0>, rot=<0 0 0 1>}
 CTRL_expressions_jawChinRaiseUR {type=Curve, loc=<0 0 0>, rot=<0 0 0 1>}
 CTRL_expressions_jawOpenExtreme {type=Curve, loc=<0 0 0>, rot=<0 0 0 1>}
 CTRL_expressions_teethUpU {type=Curve, loc=<0 0 0>, rot=<0 0 0 1>}
 CTRL_expressions_teethUpD {type=Curve, loc=<0 0 0>, rot=<0 0 0 1>}
 CTRL_expressions_teethDownU {type=Curve, loc=<0 0 0>, rot=<0 0 0 1>}
 CTRL_expressions_teethDownD {type=Curve, loc=<0 0 0>, rot=<0 0 0 1>}
 CTRL_expressions_teethLeftU {type=Curve, loc=<0 0 0>, rot=<0 0 0 1>}
 CTRL_expressions_teethLeftD {type=Curve, loc=<0 0 0>, rot=<0 0 0 1>}
 CTRL_expressions_teethRightU {type=Curve, loc=<0 0 0>, rot=<0 0 0 1>}
 CTRL_expressions_teethRightD {type=Curve, loc=<0 0 0>, rot=<0 0 0 1>}
 CTRL_expressions_teethFwdU {type=Curve, loc=<0 0 0>, rot=<0 0 0 1>}
 CTRL_expressions_teethFwdD {type=Curve, loc=<0 0 0>, rot=<0 0 0 1>}
 CTRL_expressions_teethBackU {type=Curve, loc=<0 0 0>, rot=<0 0 0 1>}
 CTRL_expressions_teethBackD {type=Curve, loc=<0 0 0>, rot=<0 0 0 1>}
 CTRL_expressions_teethRotateFwdU {type=Curve, loc=<0 0 0>, rot=<0 0 0 1>}
 CTRL_expressions_teethRotateFwdD {type=Curve, loc=<0 0 0>, rot=<0 0 0 1>}
 CTRL_expressions_teethRotateBackU {type=Curve, loc=<0 0 0>, rot=<0 0 0 1>}
 CTRL_expressions_teethRotateBackD {type=Curve, loc=<0 0 0>, rot=<0 0 0 1>}
 DisableFaceOverride {type=Curve, loc=<0 0 0>, rot=<0 0 0 1>}
 is_3L {type=Curve, loc=<0 0 0>, rot=<0 0 0 1>}
]
